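# Physics asset for the MicroLeaf pickaxe mesh, set up for dangling leaves that can swing.
# Nine leaf chains (A to I) each run from a base capsule through three thin dynamic capsules.
# Constraints link each segment to the one before it with moderate swing and almost no twist.
# The root joints of leaves E, F and G get wider swing and slightly more twist than the others.
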PhysicsAsset /BRCosmetics/Weapons/FORT_Melee/Pickaxe_MicroLeaf/Meshes/MicroLeaf_Axe_Physics
bodies {
 base_leaf_A [{shapeType=Capsule, length=0.505, radius=0.505, shapeName=base_leaf_A_capsule, rotation=<rot 0 -90 0>}]
 base_leaf_B [{shapeType=Capsule, length=0.505, radius=0.505, shapeName=base_leaf_B_capsule, rotation=<rot 0 -90 0>}]
 base_leaf_C [{shapeType=Capsule, length=0.505, radius=0.505, shapeName=base_leaf_C_capsule, rotation=<rot 0 -90 0>}]
 base_leaf_D [{shapeType=Capsule, length=0.505, radius=0.505, shapeName=base_leaf_D_capsule, rotation=<rot 0 -90 0>}]
 base_leaf_E [{shapeType=Capsule, length=0.505, radius=0.505, shapeName=base_leaf_E_capsule, rotation=<rot 0 -90 0>}]
 base_leaf_F [{shapeType=Capsule, length=0.505, radius=0.505, shapeName=base_leaf_F_capsule, rotation=<rot 0 -90 0>}]
 base_leaf_G [{shapeType=Capsule, length=0.505, radius=0.505, shapeName=base_leaf_G_capsule, rotation=<rot 0 -90 0>}]
 base_leaf_H [{shapeType=Capsule, length=0.505, radius=0.505, shapeName=base_leaf_H_capsule, rotation=<rot 0 -90 0>}]
 base_leaf_I [{shapeType=Capsule, length=0.505, radius=0.505, shapeName=base_leaf_I_capsule, rotation=<rot 0 -90 0>}]
 dyn_leaf_A_01 [{shapeType=Capsule, length=1, radius=0.2, shapeName=dyn_leaf_A_01_capsule, rotation=<rot 90 0 0>}]
 dyn_leaf_A_02 [{shapeType=Capsule, length=1, radius=0.2, shapeName=dyn_leaf_A_02_capsule, rotation=<rot 90 0 0>}]
 dyn_leaf_A_03 [{shapeType=Capsule, length=1, radius=0.2, shapeName=dyn_leaf_A_03_capsule, rotation=<rot 90 0 0>}]
 dyn_leaf_B_01 [{shapeType=Capsule, length=1, radius=0.2, shapeName=dyn_leaf_B_01_capsule, rotation=<rot 90 0 0>}]
 dyn_leaf_B_02 [{shapeType=Capsule, length=1, radius=0.2, shapeName=dyn_leaf_B_02_capsule, rotation=<rot 90 0 0>}]
 dyn_leaf_B_03 [{shapeType=Capsule, length=1, radius=0.2, shapeName=dyn_leaf_B_03_capsule, rotation=<rot 90 0 0>}]
 dyn_leaf_C_01 [{shapeType=Capsule, length=1, radius=0.2, shapeName=dyn_leaf_C_01_capsule, rotation=<rot 90 0 0>}]
 dyn_leaf_C_02 [{shapeType=Capsule, length=1, radius=0.2, shapeName=dyn_leaf_C_02_capsule, rotation=<rot 90 0 0>}]
 dyn_leaf_C_03 [{shapeType=Capsule, length=1, radius=0.2, shapeName=dyn_leaf_C_03_capsule, rotation=<rot 90 0 0>}]
 dyn_leaf_D_01 [{shapeType=Capsule, length=1, radius=0.2, shapeName=dyn_leaf_D_01_capsule, rotation=<rot 90 0 0>}]
 dyn_leaf_D_02 [{shapeType=Capsule, length=1, radius=0.2, shapeName=dyn_leaf_D_02_capsule, rotation=<rot 90 0 0>}]
 dyn_leaf_D_03 [{shapeType=Capsule, length=1, radius=0.2, shapeName=dyn_leaf_D_03_capsule, rotation=<rot 90 0 0>}]
 dyn_leaf_E_01 [{shapeType=Capsule, length=1, radius=0.2, shapeName=dyn_leaf_E_01_capsule, rotation=<rot 90 0 0>}]
 dyn_leaf_E_02 [{shapeType=Capsule, length=1, radius=0.2, shapeName=dyn_leaf_E_02_capsule, rotation=<rot 90 0 0>}]
 dyn_leaf_E_03 [{shapeType=Capsule, length=1, radius=0.2, shapeName=dyn_leaf_E_03_capsule, rotation=<rot 90 0 0>}]
 dyn_leaf_F_01 [{shapeType=Capsule, length=1, radius=0.2, shapeName=dyn_leaf_F_01_capsule, rotation=<rot 90 0 0>}]
 dyn_leaf_F_02 [{shapeType=Capsule, length=1, radius=0.2, shapeName=dyn_leaf_F_02_capsule, rotation=<rot 90 0 0>}]
 dyn_leaf_F_03 [{shapeType=Capsule, length=1, radius=0.2, shapeName=dyn_leaf_F_03_capsule, rotation=<rot 90 0 0>}]
 dyn_leaf_G_01 [{shapeType=Capsule, length=1, radius=0.2, shapeName=dyn_leaf_G_01_capsule, rotation=<rot 90 0 0>}]
 dyn_leaf_G_02 [{shapeType=Capsule, length=1, radius=0.2, shapeName=dyn_leaf_G_02_capsule, rotation=<rot 90 0 0>}]
 dyn_leaf_G_03 [{shapeType=Capsule, length=1, radius=0.2, shapeName=dyn_leaf_G_03_capsule, rotation=<rot 90 0 0>}]
 dyn_leaf_H_01 [{shapeType=Capsule, length=1, radius=0.2, shapeName=dyn_leaf_H_01_capsule, rotation=<rot 90 0 0>}]
 dyn_leaf_H_02 [{shapeType=Capsule, length=1, radius=0.2, shapeName=dyn_leaf_H_02_capsule, rotation=<rot 90 0 0>}]
 dyn_leaf_H_03 [{shapeType=Capsule, length=1, radius=0.2, shapeName=dyn_leaf_H_03_capsule, rotation=<rot 90 0 0>}]
 dyn_leaf_I_01 [{shapeType=Capsule, length=1, radius=0.2, shapeName=dyn_leaf_I_01_capsule, rotation=<rot 90 0 0>}]
 dyn_leaf_I_02 [{shapeType=Capsule, length=1, radius=0.2, shapeName=dyn_leaf_I_02_capsule, rotation=<rot 90 0 0>}]
 dyn_leaf_I_03 [{shapeType=Capsule, length=1, radius=0.2, shapeName=dyn_leaf_I_03_capsule, rotation=<rot 90 0 0>}]
}
constraints [
 {bone1Name=dyn_leaf_A_01, bone2Name=base_leaf_A, swing1LimitDegrees=45, swing2LimitDegrees=45, twistLimitDegrees=1}
 {bone1Name=dyn_leaf_A_02, bone2Name=dyn_leaf_A_01, swing1LimitDegrees=45, swing2LimitDegrees=45, twistLimitDegrees=1}
 {bone1Name=dyn_leaf_A_03, bone2Name=dyn_leaf_A_02, swing1LimitDegrees=45, swing2LimitDegrees=45, twistLimitDegrees=1}
 {bone1Name=dyn_leaf_B_01, bone2Name=base_leaf_B, swing1LimitDegrees=45, swing2LimitDegrees=45, twistLimitDegrees=1}
 {bone1Name=dyn_leaf_B_02, bone2Name=dyn_leaf_B_01, swing1LimitDegrees=45, swing2LimitDegrees=45, twistLimitDegrees=1}
 {bone1Name=dyn_leaf_B_03, bone2Name=dyn_leaf_B_02, swing1LimitDegrees=45, swing2LimitDegrees=45, twistLimitDegrees=1}
 {bone1Name=dyn_leaf_C_01, bone2Name=base_leaf_C, swing1LimitDegrees=45, swing2LimitDegrees=45, twistLimitDegrees=1}
 {bone1Name=dyn_leaf_C_02, bone2Name=dyn_leaf_C_01, swing1LimitDegrees=45, swing2LimitDegrees=45, twistLimitDegrees=1}
 {bone1Name=dyn_leaf_C_03, bone2Name=dyn_leaf_C_02, swing1LimitDegrees=45, swing2LimitDegrees=45, twistLimitDegrees=1}
 {bone1Name=dyn_leaf_D_01, bone2Name=base_leaf_D, swing1LimitDegrees=45, swing2LimitDegrees=45, twistLimitDegrees=1}
 {bone1Name=dyn_leaf_D_02, bone2Name=dyn_leaf_D_01, swing1LimitDegrees=45, swing2LimitDegrees=45, twistLimitDegrees=1}
 {bone1Name=dyn_leaf_D_03, bone2Name=dyn_leaf_D_02, swing1LimitDegrees=45, swing2LimitDegrees=45, twistLimitDegrees=1}
 {bone1Name=dyn_leaf_E_01, bone2Name=base_leaf_E, swing1LimitDegrees=75, swing2LimitDegrees=75, twistLimitDegrees=2}
 {bone1Name=dyn_leaf_E_02, bone2Name=dyn_leaf_E_01, swing1LimitDegrees=45, swing2LimitDegrees=45, twistLimitDegrees=1}
 {bone1Name=dyn_leaf_E_03, bone2Name=dyn_leaf_E_02, swing1LimitDegrees=45, swing2LimitDegrees=45, twistLimitDegrees=1}
 {bone1Name=dyn_leaf_F_01, bone2Name=base_leaf_F, swing1LimitDegrees=75, swing2LimitDegrees=75, twistLimitDegrees=2}
 {bone1Name=dyn_leaf_F_02, bone2Name=dyn_leaf_F_01, swing1LimitDegrees=45, swing2LimitDegrees=45, twistLimitDegrees=1}
 {bone1Name=dyn_leaf_F_03, bone2Name=dyn_leaf_F_02, swing1LimitDegrees=45, swing2LimitDegrees=45, twistLimitDegrees=1}
 {bone1Name=dyn_leaf_G_01, bone2Name=base_leaf_G, swing1LimitDegrees=75, swing2LimitDegrees=75, twistLimitDegrees=2}
 {bone1Name=dyn_leaf_G_02, bone2Name=dyn_leaf_G_01, swing1LimitDegrees=45, swing2LimitDegrees=45, twistLimitDegrees=1}
 {bone1Name=dyn_leaf_G_03, bone2Name=dyn_leaf_G_02, swing1LimitDegrees=45, swing2LimitDegrees=45, twistLimitDegrees=1}
 {bone1Name=dyn_leaf_H_01, bone2Name=base_leaf_H, swing1LimitDegrees=45, swing2LimitDegrees=45, twistLimitDegrees=1}
 {bone1Name=dyn_leaf_H_02, bone2Name=dyn_leaf_H_01, swing1LimitDegrees=45, swing2LimitDegrees=45, twistLimitDegrees=1}
 {bone1Name=dyn_leaf_H_03, bone2Name=dyn_leaf_H_02, swing1LimitDegrees=45, swing2LimitDegrees=45, twistLimitDegrees=1}
 {bone1Name=dyn_leaf_I_01, bone2Name=base_leaf_I, swing1LimitDegrees=45, swing2LimitDegrees=45, twistLimitDegrees=1}
 {bone1Name=dyn_leaf_I_02, bone2Name=dyn_leaf_I_01, swing1LimitDegrees=45, swing2LimitDegrees=45, twistLimitDegrees=1}
 {bone1Name=dyn_leaf_I_03, bone2Name=dyn_leaf_I_02, swing1LimitDegrees=45, swing2LimitDegrees=45, twistLimitDegrees=1}
]
