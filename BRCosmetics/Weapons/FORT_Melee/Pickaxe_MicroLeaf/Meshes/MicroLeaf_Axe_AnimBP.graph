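# Animation Blueprint for the MicroLeaf pickaxe that adds rigid body physics motion on top of its input pose.
# Two RigidBody nodes share one physics asset, one simulating in handle-bone space, the other in world space at low alpha.
# An unwired Blend Poses by bool chooses between them.
# The update event flips GravityOverride on a skydiving check, but the graph never reads it, and both simulations override gravity to zero.
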AnimBlueprint /BRCosmetics/Weapons/FORT_Melee/Pickaxe_MicroLeaf/Meshes/MicroLeaf_Axe_AnimBP
target_skeleton=&/MicroLeaf_Axe_Skeleton
graphs [
 AnimGraph {nodes [
   n1 = RigidBody!(ComponentPose=Local To Component(LocalPose=Use cached pose 'Input pose'()), Alpha, OverrideWorldGravity:Vector="0,0,0", ExternalForce) -> (Pose) {
    bClampLinearTranslationLimitToRefPose=true
    bOverrideWorldGravity=true
    overridePhysicsAsset=&/MicroLeaf_Axe_Physics
    simulationSpace=BaseBoneSpace
    baseBoneRef {boneName=Melee_TwoHanded_Handle}
    componentAppliedLinearAccClamp=<0 0 0>
    simSpaceSettings {
     dampingAlpha=0
     maxAngularAcceleration=15
     maxAngularVelocity=3
     maxLinearAcceleration=2000
     maxLinearVelocity=600
     velocityScaleZ=0.5
     worldAlpha=1
     externalLinearDragV {x=0.5, z=0.5}
    }
   }
   n2 = RigidBody(ComponentPose=Local To Component(LocalPose=Use cached pose 'Input pose'()), Alpha=0.2, OverrideWorldGravity="0,0,0") {
    bClampLinearTranslationLimitToRefPose=true
    bOverrideWorldGravity=true
    overridePhysicsAsset=&MicroLeaf_Axe_Physics
    simulationSpace=WorldSpace
    baseBoneRef {boneName=Melee_TwoHanded_Handle}
    componentAppliedLinearAccClamp=<0 0 0>
    simSpaceSettings {dampingAlpha=0, maxAngularAcceleration=1000, maxAngularVelocity=120, maxLinearVelocity=1000, worldAlpha=1, externalLinearDragV {x=50, z=500}}
   }
   Output Pose(Result=Blend Poses by bool(BlendPose_0=Component To Local(ComponentPose=@n2), BlendPose_1=Component To Local(ComponentPose=@n1))) {layerGroup=None}
   AnimGraphNode_SubInput_0 = Input Pose()
   Save cached pose 'Input pose'(Pose=@AnimGraphNode_SubInput_0) {cacheName=Input pose}
  ]}
 EventGraph {nodes [
   Event BlueprintUpdateAnimation()
   > n1 = Branch(Condition=Get bIsSkyDiving())
   > Set GravityOverride(GravityOverride="0,0,980")
   Set GravityOverride(execute=@n1.else, GravityOverride="0,0,-980")
   TryGetPawnOwner()
  ]}
]
variables [GravityOverride {type=Vector}]
